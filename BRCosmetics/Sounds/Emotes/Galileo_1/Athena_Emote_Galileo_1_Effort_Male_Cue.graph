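# Sound cue for the male effort sound of the Galileo_1 emote in Athena.
# A mixer plays two random picks together: one of three male effort sounds at reduced volume and one of two reach sounds at full volume.
# It uses emote attenuation, the emote music sound class, concurrency settings that name stop-oldest limiting and an emote foley rule, and a send to the music audio-analysis submix.
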SoundCue /BRCosmetics/Sounds/Emotes/Galileo_1/Athena_Emote_Galileo_1_Effort_Male_Cue
root=Mixer_0
nodes [
 {id=Mixer_0, children [Random_2,Random_1], properties {inputVolume [0.35,1]}}
 {id=Random_1, children [WavePlayer_6,WavePlayer_7], properties {weights [1,1]}}
 {id=Random_2, children [WavePlayer_0,WavePlayer_4,WavePlayer_5], properties {weights [1,1,1]}}
 {id=WavePlayer_0, properties {soundWaveAssetPtr=&/Emote_Galileo_1_Male_Effort_01}}
 {id=WavePlayer_4, properties {soundWaveAssetPtr=&/Emote_Galileo_1_Male_Effort_02}}
 {id=WavePlayer_5, properties {soundWaveAssetPtr=&/Emote_Galileo_1_Male_Effort_03}}
 {id=WavePlayer_6, properties {soundWaveAssetPtr=/Game/Athena/Sounds/Emotes/Galileo_1/Emote_Galileo_1_Reach_01}}
 {id=WavePlayer_7, properties {soundWaveAssetPtr=/Game/Athena/Sounds/Emotes/Galileo_1/Emote_Galileo_1_Reach_02}}
]
properties {
 attenuationSettings=&Fort_Emotes_Athena_Att
 maxDistance=1600
 soundClassObject=&Fort_Music_Emote
 volumeMultiplier=1
 concurrencyOverrides {maxCount=3, resolutionRule=StopOldest}
 concurrencySet [/Game/Sounds/Concurrency_Rules/Athena_EmoteFoley_Conc]
 soundSubmixSends [{minSendLevel=0, soundSubmix=&Fortnite_AudioAnalysis_Music}]
}
